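# Physics asset for the female medium Bandolier Slurp head, providing collision bodies and joints for its dynamic hair bones.
# A head capsule anchors five small base boxes through angularly free joints, each starting a hair chain.
# The long left chain of boxes and two single-capsule right strands use tight swing and twist limits.
# The two short top slurp chains leave swing and twist free throughout.
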PhysicsAsset /BRCosmetics/Characters/Player/Female/Medium/Heads/F_MED_Bandolier_Slurp_Head/Meshes/F_MED_Bandolier_Slurp_Head_Physics
bodies {
 dyn_hairA_01_Base_r [{shapeType=Box, extentX=1.01, extentY=1.01, extentZ=1.01, rotation=<rot 0 -90 0>}]
 dyn_hairA_01_r [{shapeType=Capsule, length=5.155, radius=1.1827, center=<3.0755 -0.2988 -0.0238>, rotation=<rot 0 -86.1603 -87.0556>}]
 dyn_hairB_01_Base_r [{shapeType=Box, extentX=1.01, extentY=1.01, extentZ=1.01, rotation=<rot 0 -90 0>}]
 dyn_hairB_01_r [{shapeType=Capsule, length=4.9215, radius=1.0996, center=<3.1115 0.2174 0.0157>, rotation=<rot 0 87.0217 -89.7818>}]
 dyn_hairSlurp_01_Base_l [{shapeType=Box, extentX=1.01, extentY=1.01, extentZ=1.01, rotation=<rot 0 -90 0>}]
 dyn_hairSlurp_01_l [{shapeType=Box, extentX=2.5044, extentY=9.3092, extentZ=16.1706, rotation=<rot 0 90 0>}]
 dyn_hairTopSlurpA_01 [{shapeType=Box, extentX=2.3053, extentY=0.5496, extentZ=2.468, center=<-0.6611 0.0954 -0.2732>, rotation=<rot 33.3087 -90 12.6701>}]
 dyn_hairTopSlurpA_01_Base [{shapeType=Box, extentX=1.01, extentY=1.01, extentZ=1.01, rotation=<rot 0 -90 0>}]
 dyn_hairTopSlurpA_02 [{shapeType=Box, extentX=1.01, extentY=1.01, extentZ=1.01, center=<-2.1312 2.6067 -1.1315>, rotation=<rot -0.1468 -90 -1.2237>}]
 dyn_hairTopSlurpB_01 [{shapeType=Box, extentX=1.5447, extentY=1.6317, extentZ=2.2294, center=<-0.0732 1.4136 0.8515>, rotation=<rot 0 -11.4309 -76.5169>}]
 dyn_hairTopSlurpB_01_Base [{shapeType=Box, extentX=1.01, extentY=1.01, extentZ=1.01, rotation=<rot 0 -90 0>}]
 dyn_hairTopSlurpB_02 [{shapeType=Box, extentX=1.01, extentY=1.01, extentZ=1.01, center=<-1.5306 1.9486 1.2547>, rotation=<rot 0 2.8645 -71.4955>}]
 dyn_hair_01_l [{shapeType=Box, extentX=2.5044, extentY=5.2392, extentZ=15.4284, center=<2.0643 0 0>, rotation=<rot 0 90 0>}]
 dyn_hair_02_l [{shapeType=Box, extentX=2.5044, extentY=8.0752, extentZ=15.7715, center=<3.4417 0 0>, rotation=<rot 0 90 0>}]
 dyn_hair_03_l [{shapeType=Box, extentX=2.5044, extentY=3.6106, extentZ=8.7503, center=<1.4881 -0.7611 0>, rotation=<rot 0 62.912 0>}]
 dyn_hair_04_l [{shapeType=Box, extentX=2.991, extentY=5.5736, extentZ=7.9462, center=<2.0744 0.8312 0.0001>, rotation=<rot 0 90 0>}]
 head [{shapeType=Capsule, length=11.5599, radius=5.8421, center=<3.2009 1.3916 0>, rotation=<rot 90 0 0>}]
}
constraints [
 {
  bone1Name=dyn_hairSlurp_01_Base_l
  bone2Name=head
  swing1LimitDegrees=45
  swing1Motion=Free
  swing2LimitDegrees=45
  swing2Motion=Free
  twistLimitDegrees=45
  twistMotion=Free
 }
 {
  bone1Name=dyn_hairTopSlurpB_01_Base
  bone2Name=head
  swing1LimitDegrees=45
  swing1Motion=Free
  swing2LimitDegrees=45
  swing2Motion=Free
  twistLimitDegrees=45
  twistMotion=Free
 }
 {
  bone1Name=dyn_hairTopSlurpA_01_Base
  bone2Name=head
  swing1LimitDegrees=45
  swing1Motion=Free
  swing2LimitDegrees=45
  swing2Motion=Free
  twistLimitDegrees=45
  twistMotion=Free
 }
 {
  bone1Name=dyn_hairB_01_Base_r
  bone2Name=head
  swing1LimitDegrees=45
  swing1Motion=Free
  swing2LimitDegrees=45
  swing2Motion=Free
  twistLimitDegrees=45
  twistMotion=Free
 }
 {
  bone1Name=dyn_hairA_01_Base_r
  bone2Name=head
  swing1LimitDegrees=45
  swing1Motion=Free
  swing2LimitDegrees=45
  swing2Motion=Free
  twistLimitDegrees=45
  twistMotion=Free
 }
 {bone1Name=dyn_hairSlurp_01_l, bone2Name=dyn_hairSlurp_01_Base_l, swing1LimitDegrees=20, swing2LimitDegrees=20, twistLimitDegrees=8}
 {bone1Name=dyn_hair_01_l, bone2Name=dyn_hairSlurp_01_l, swing1LimitDegrees=20, swing2LimitDegrees=20, twistLimitDegrees=8}
 {bone1Name=dyn_hair_02_l, bone2Name=dyn_hair_01_l, swing1LimitDegrees=20, swing2LimitDegrees=20, twistLimitDegrees=8}
 {bone1Name=dyn_hair_03_l, bone2Name=dyn_hair_02_l, swing1LimitDegrees=20, swing2LimitDegrees=20, twistLimitDegrees=8}
 {bone1Name=dyn_hair_04_l, bone2Name=dyn_hair_03_l, swing1LimitDegrees=20, swing2LimitDegrees=20, twistLimitDegrees=8}
 {
  bone1Name=dyn_hairTopSlurpB_01
  bone2Name=dyn_hairTopSlurpB_01_Base
  swing1LimitDegrees=45
  swing1Motion=Free
  swing2LimitDegrees=45
  swing2Motion=Free
  twistLimitDegrees=45
  twistMotion=Free
 }
 {
  bone1Name=dyn_hairTopSlurpB_02
  bone2Name=dyn_hairTopSlurpB_01
  swing1LimitDegrees=45
  swing1Motion=Free
  swing2LimitDegrees=45
  swing2Motion=Free
  twistLimitDegrees=45
  twistMotion=Free
 }
 {
  bone1Name=dyn_hairTopSlurpA_01
  bone2Name=dyn_hairTopSlurpA_01_Base
  swing1LimitDegrees=45
  swing1Motion=Free
  swing2LimitDegrees=45
  swing2Motion=Free
  twistLimitDegrees=45
  twistMotion=Free
 }
 {
  bone1Name=dyn_hairTopSlurpA_02
  bone2Name=dyn_hairTopSlurpA_01
  swing1LimitDegrees=45
  swing1Motion=Free
  swing2LimitDegrees=45
  swing2Motion=Free
  twistLimitDegrees=45
  twistMotion=Free
 }
 {bone1Name=dyn_hairB_01_r, bone2Name=dyn_hairB_01_Base_r, swing1LimitDegrees=20, swing2LimitDegrees=20, twistLimitDegrees=20}
 {bone1Name=dyn_hairA_01_r, bone2Name=dyn_hairA_01_Base_r, swing1LimitDegrees=20, swing2LimitDegrees=20, twistLimitDegrees=20}
]
